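# Sound cue for searching a fireplace loot container, routed to the searching foley sound class with prop-search attenuation.
# A mixer layers a random fireplace search wave with a quieter random firepit search wave, each varied slightly in pitch and volume.
# The fireplace branch also feeds a fixed delay into the mixer at half volume, adding a later, softer search sound.
# A concurrency override prevents new instances once the limit is reached.
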SoundCue /Figment_S10/Sounds/Foley_Loot/Containers/Fireplace/Fireplace_Search_Cue
root=Mixer_4
nodes [
 {id=Delay_15, children [Modulator_10], properties {delayMax=0.85, delayMin=0.85}}
 {id=Mixer_4, children [Modulator_10,Modulator_11,Delay_15], properties {inputVolume [1,0.7,0.5]}}
 {id=Modulator_10, children [Random_9], properties {pitchMax=1.07, pitchMin=0.92, volumeMax=1.07, volumeMin=0.92}}
 {id=Modulator_11, children [Random_10], properties {pitchMax=1.07, pitchMin=0.92, volumeMax=1.07, volumeMin=0.92}}
 {id=Random_10, children [WavePlayer_27, WavePlayer_28, WavePlayer_29], properties {weights [1,1,1]}}
 {id=Random_9, children [WavePlayer_24, WavePlayer_25, WavePlayer_26], properties {weights [1,1,1]}}
 {id=WavePlayer_24, properties {soundWaveAssetPtr=&/Fireplace_Search01}}
 {id=WavePlayer_25, properties {soundWaveAssetPtr=&/Fireplace_Search02}}
 {id=WavePlayer_26, properties {soundWaveAssetPtr=&/Fireplace_Search08}}
 {id=WavePlayer_27, properties {soundWaveAssetPtr=/Figment_S10/Sounds/Foley_Loot/Containers/Firepit/FirePit_Search13}}
 {id=WavePlayer_28, properties {soundWaveAssetPtr=/Figment_S10/Sounds/Foley_Loot/Containers/Firepit/FirePit_Search07}}
 {id=WavePlayer_29, properties {soundWaveAssetPtr=/Figment_S10/Sounds/Foley_Loot/Containers/Firepit/FirePit_Search12}}
]
properties {
 attenuationSettings=/Game/Sounds/Attenuation/Fort_Prop_Search_Att
 bOverrideConcurrency=true
 maxDistance=1300
 soundClassObject=/Figment_S10/Sounds/FortSoundClassesAndModes/Fort_Searching_Foley
 volumeMultiplier=1.2
 concurrencyOverrides {resolutionRule=PreventNew}
}
